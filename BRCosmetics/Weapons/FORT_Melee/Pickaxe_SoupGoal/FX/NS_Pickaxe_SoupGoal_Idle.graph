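# Niagara idle effect for the SoupGoal pickaxe that spawns short-lived mesh lightning bolts, fading from bright cyan-blue toward magenta.
# A system timeline drives a spawn rate that tapers to zero, alongside a per-frame burst gated by the inverse of a value-change flag.
# Each particle picks one of four variants from User.LightningOrder by ID, with random mirroring and inversion fed to the material through dynamic parameters.
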NiagaraSystem /BRCosmetics/Weapons/FORT_Melee/Pickaxe_SoupGoal/FX/NS_Pickaxe_SoupGoal_Idle
dependencies {
 usedDataInterfaces [
  &NiagaraDataInterfaceCurve
  &NiagaraDataInterfaceArrayFloat
  &NiagaraDataInterfaceMeshRendererInfo
  &NiagaraDataInterfaceColorCurve
 ]
 usedDynamicInputs [
  &SimulationPosition
  &Add_Integer
  &Multiply_Int
  &RandomRangeInt
 ]
 usedModules [
  &SystemState
  /Game/Effects/Niagara/Modules/Logic/UpdateValueStateChange
  &Timeline
  &self:SystemSpawnScript.SystemScriptSource.SystemScriptGraph.NiagaraNodeAssignment_0.SetVariables_#1
 ]
}
emitters {Lightning {
  emitterUpdateScript [
   &EmitterState(
    Inactive Response=::NewEnumerator1
    Life Cycle Mode=::NewEnumerator1
    Loop Behavior=::NewEnumerator1
    Loop Duration Mode=::NewEnumerator1
    UseLoopDelay=false
   )
   &SpawnPerFrame(
    Spawn:NiagaraBool=&InvertBool(Boolean:NiagaraBool=$System.UpdateValueStateChange.Changed)
    Spawn Count=4
    Use Spawn Probability=false
   )
   &SpawnRate(
    SpawnRate=&FloatFromCurve(
     FloatCurve=DataInterface() {propertyValues {Curve {keys [
         {tangentMode=RCTM_User, time=0, value=1}
         {arriveTangent=-0.0405, interpMode=RCIM_Cubic, leaveTangent=-0.0405, tangentMode=RCTM_User, time=0.7009, value=1}
         {arriveTangent=-8.0089, interpMode=RCIM_Cubic, leaveTangent=-8.0089, tangentMode=RCTM_User, time=1, value=0}
        ]}}}
     CurveIndex=$System.Timeline.NormalizedTime
     Scale Curve=8
    )
    Use Spawn Probability=false
   )
  ]
  particleSpawnScript [
   &InitializeParticle(
    Color=<rgba 0 0.9 3 1>
    Lifetime=&RandomRangeFloat(Minimum=0.4, Maximum=0.6)
    Mesh Scale Mode=::NewEnumerator3
    UsePositionOffset=false
    Write Material Random=true
   )
   /SetVariables(
    Particles.InvertLightning:NiagaraBool=&RandomBool()
    Particles.LightningIndex:NiagaraInt32=&ModuloInt(
     A=&SelectIntFromArray(
      Direct Array Index=&ReturnParticleIDAttributes()
      Direct Array Mode=::NewEnumerator1
      Int Selection Array=$User.LightningOrder
     )
     B=4
    )
    Particles.MirrorLightning:NiagaraBool=&RandomBool()
   ) {isSetParameters=true}
   /SetVariables(Particles.Scale=&MakeCustomVectorFromBool(Bool=$Particles.MirrorLightning, True Vector=<-1 1 1>, False Vector=<1 1 1>)) {isSetParameters=true}
  ]
  particleUpdateScript [
   &DynamicMaterialParameters(
    Index 0 Param 1=&RandomRangeFloat(Minimum=7, Maximum=7)
    Index 0 Param 2=&Divide_Float(
     A=&RandomRangeFloat(Minimum=1, Maximum=1.5)
     B=&SelectFloatFromArray(Direct Array Index=$Particles.LightningIndex, Float Selection Array=DataInterface() {propertyValues {FloatData [2.6,2.3,2,1.5]}})
    )
    Index 0 Param 3=&FloatFromCurve(FloatCurve=DataInterface() {propertyValues {Curve {keys [<at 0 1>,<at 0.1 0.375>,<at 0.5 0.05>,<at 1 0>]}}})
    Index 0 Param 4=&MakeCustomFloatFromBool(Bool=$Particles.InvertLightning, True Float=1, False Float=0)
    Index 1 Param 1=&MakeFloatFromVector(Vector=$Particles.Scale, Channel=::NewEnumerator0)
    Index 1 Param 2=&MakeFloatFromVector(Vector=$Particles.Scale, Channel=::NewEnumerator1)
    Index 1 Param 3=&MakeFloatFromVector(Vector=$Particles.Scale, Channel=::NewEnumerator2)
    Module.Index0Param0Write=true
    Module.Index0Param1Write=true
    Module.Index0Param2Write=true
    Module.Index0Param3Write=true
    Module.Index1Param0Write=true
    Module.Index1Param1Write=true
    Module.Index1Param2Write=true
    Module.Index1Param3Write=true
    Write Parameter Index 1=true
   )
   &ParticleState()
   &Color(
    Color=&ColorFromCurve(ColorCurve=DataInterface() {propertyValues {
       AlphaCurve {keys [<at 0 0>, {arriveTangent=0.3873, interpMode=RCIM_Cubic, leaveTangent=0.3873, time=0.0998, value=1}, <at 0.7746 0.3>, <at 1 0 RCIM_Cubic>]}
       BlueCurve {keys [<at 0 66 RCIM_Cubic>,<at 0.2919 3>,<at 1 1 RCIM_Cubic>]}
       GreenCurve {keys [<at 0 32.9999 RCIM_Cubic>,<at 0.2919 0.35>,<at 1 0 RCIM_Cubic>]}
       RedCurve {keys [<at 0 0 RCIM_Cubic>,<at 0.2919 0>,<at 1 1 RCIM_Cubic>]}
      }})
    Scale Color=<2 2 2>
   )
  ]
  rendererClasses [&NiagaraMeshRendererProperties]
  renderers [{rendererClass=&NiagaraMeshRendererProperties, rendererIndex=0}]
 }}
summary {
 systemName=NS_Pickaxe_SoupGoal_Idle
 emitters [{emitterName=Lightning, rendererClasses [&NiagaraMeshRendererProperties]}]
 userVariables [
  User.bIsFrontEnd {type=NiagaraBool}
  User.bIsFrontEndPreview {type=NiagaraBool}
  User.DebugReactivity {type=NiagaraBool}
  User.LightningOrder {
   type=NiagaraDataInterfaceArrayInt32
   description=""
   defaultValue {
    struct=NiagaraExt_VariableValue_DataInterface
    value {
     dataInterface=&self:NiagaraDataInterfaceArrayInt32_0
     dataInterfaceClass=&NiagaraDataInterfaceArrayInt32
    }
   }
  }
  User.ReactivityCounter {type=NiagaraInt32}
  User.ReactivityDuration {type=NiagaraFloat, defaultValue=4, description=""}
 ]
}
