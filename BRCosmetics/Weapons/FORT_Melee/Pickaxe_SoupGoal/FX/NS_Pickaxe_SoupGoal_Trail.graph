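# Niagara weapon trail for the SoupGoal pickaxe, pairing a base swing ribbon with a reactive lightning ribbon.
# FN_AnimTrail builds a short-lived ribbon between two skeletal mesh bones or sockets, tinted by a user color curve.
# LightningTrail samples the same mesh each frame, feeding three ribbon renderers with tip fading.
# Normalized reactivity time narrows the lightning, shifts it toward base trail color and crossfades separate reactive and base colors late in the window.
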NiagaraSystem /BRCosmetics/Weapons/FORT_Melee/Pickaxe_SoupGoal/FX/NS_Pickaxe_SoupGoal_Trail
dependencies {
 usedDataInterfaces [&NiagaraDataInterfaceCurve, &NiagaraDataInterfaceSkeletalMesh]
 usedDynamicInputs [&ClampFloat, &Divide_Float, &Subtract_Float]
 usedModules [
  &self:SystemSpawnScript.SystemScriptSource.SystemScriptGraph.NiagaraNodeAssignment_0.SetVariables_#1
  &SystemState
  &self:SystemSpawnScript.SystemScriptSource.SystemScriptGraph.NiagaraNodeAssignment_1.SetVariables_#2
  &self:SystemSpawnScript.SystemScriptSource.SystemScriptGraph.NiagaraNodeAssignment_3.SetVariables_#3
  &self:SystemSpawnScript.SystemScriptSource.SystemScriptGraph.NiagaraNodeAssignment_2.SetVariables_#4
 ]
}
emitters {
 FN_AnimTrail {
  emitterSpawnScript [&FN_RibbonIDOffsetUtility()]
  emitterUpdateScript [
   &EmitterState()
   &FN_SpawnAnimTrail()
  ]
  particleSpawnScript [
   &FN_InitializeAnimTrailBoneIndecies(First Bone or Socket Index=1, Second Bone or Socket Index=0)
   &FN_InitializeAnimTrail(AnimTrailWidth=1.7, Color=<rgba 0.5828 0.1458 1 1>, Lifetime=0.14, SkeletalMesh=$System.Skeletal Mesh)
  ]
  particleUpdateScript [
   &SolveForcesAndVelocity(Clamp Velocity=false, Limit Acceleration=false)
   &FN_UpdateAnimTrail(SkeletalMesh=$System.Skeletal Mesh) {module=FN_UpdateAnimTrail001}
   &ScaleColor(
    Scale Alpha=&FloatFromCurve(
     FloatCurve=DataInterface() {propertyValues {Curve {keys [{arriveTangent=-3.3248, interpMode=RCIM_Cubic, time=0, value=1}, {interpMode=RCIM_Cubic, leaveTangent=-9.9298, time=1, value=0}]}}}
     Scale Curve=0.3
    )
    Scale RGB=&VectorFromFloat(Value=1)
    ScaleA=true
    ScaleRGB=true
   )
   &Color(
    Color=&ColorFromCurve(ColorCurve=$User.BaseTrailColorCurve)
    Scale Alpha=0.333
    Scale Color=&MakeVectorFromLinearColorRGB(Color=&MultiplyLinearColor(
      Linear Color=$Particles.Initial.Color
      Scale Factor=&MakeVector4_FromLinearColor(LinearColor=$System.TODColorMultiplier)
     ))
   )
  ]
  rendererClasses [&NiagaraRibbonRendererProperties]
  renderers [{rendererClass=&NiagaraRibbonRendererProperties}]
 }
 LightningTrail {
  emitterUpdateScript [
   &EmitterState()
   &SpawnPerFrame(Use Spawn Probability=false)
  ]
  particleSpawnScript [
   &InitializeParticle(
    Color Mode=::NewEnumerator0
    Lifetime=0.2
    Ribbon Width=&Lerp_Float(A=33, B=11, Alpha=$System.ReactivityNormalizedTime)
    Ribbon Width Mode=::NewEnumerator1
    UsePositionOffset=false
    Write Material Random=true
   )
   &SkeletalMeshLocation(MaskSpawnGroup=false, Mesh Sampling Type=::NewEnumerator5, Skeletal Mesh=$System.Skeletal Mesh, Socket Sampling Mode=::NewEnumerator5)
  ]
  particleUpdateScript [
   &ParticleState()
   /SetVariables(Particles.TipFadeMultiplier:NiagaraFloat=&FloatFromCurve(
     FloatCurve=DataInterface() {propertyValues {Curve {keys [
         {tangentMode=RCTM_User, time=0, value=1}
         {arriveTangent=-0.0405, interpMode=RCIM_Cubic, leaveTangent=-0.0405, tangentMode=RCTM_User, time=0.7009, value=1}
         {arriveTangent=-8.0089, interpMode=RCIM_Cubic, leaveTangent=-8.0089, tangentMode=RCTM_User, time=1, value=0}
        ]}}}
     CurveIndex=&ReturnNormalizedExecIndex(Normalized Index Scale=&MakeCustomFloatFromBool(Bool=/Game/Effects/Niagara/DynamicInputs/State/FN_IsExecutionStateActive(), True Float=0, False Float=1))
     Scale Curve=&FloatFromCurve(
      FloatCurve=DataInterface() {propertyValues {Curve {keys [
          {arriveTangent=4.1785, interpMode=RCIM_Cubic, tangentMode=RCTM_Break, time=0, value=0}
          {arriveTangent=13.6583, interpMode=RCIM_Cubic, leaveTangent=0.0224, tangentMode=RCTM_Break, time=0.2, value=1}
         ]}}}
      CurveIndex=$Emitter.Age
     )
    )) {isSetParameters=true}
   &Color(
    Color=&LerpLinearColors(
     StartColor=&ColorFromCurve(ColorCurve=$User.LightningColorCurve)
     EndColor=&ColorFromCurve(ColorCurve=$User.BaseTrailColorCurve)
     LerpFactor=$System.ReactivityNormalizedTime
    )
    Scale Alpha=$Particles.TipFadeMultiplier
    Scale Color=&MakeVectorFromLinearColorRGB(Color=$System.TODColorMultiplier)
   )
   /SetVariables(
    Particles.BaseColor:LinearColor=&MultiplyLinearColor(
     Linear Color=$Particles.Color
     Scale Factor=&MakeFloat4FromVectorAndFloat(
      "Vector (RGB)"=&MakeVectorFromLinearColorRGB(Color=&MultiplyLinearColor(Linear Color=$System.TODColorMultiplier, Scale Factor=<1 1 1 1>))
      "Float (Alpha)"=&FloatFromCurve(
       FloatCurve=DataInterface() {propertyValues {Curve {keys [<at 0.75 0>, {arriveTangent=2.9156, time=1, value=1}]}}}
       CurveIndex=$System.ReactivityNormalizedTime
      )
     )
    )
    Particles.ReactiveColor:LinearColor=&MultiplyLinearColor(
     Linear Color=$Particles.Color
     Scale Factor=&MakeFloat4FromVectorAndFloat(
      "Vector (RGB)"=<2 2 2>
      "Float (Alpha)"=&FloatFromCurve(
       FloatCurve=DataInterface() {propertyValues {Curve {keys [<at 0.75 1>, {arriveTangent=2.9156, time=1, value=0}]}}}
       CurveIndex=$System.ReactivityNormalizedTime
      )
     )
    )
   ) {isSetParameters=true}
  ]
  rendererClasses [&NiagaraRibbonRendererProperties]
  renderers [
   {rendererClass=&NiagaraRibbonRendererProperties}
   {rendererClass=&NiagaraRibbonRendererProperties, rendererIndex=1}
   {rendererClass=&NiagaraRibbonRendererProperties, rendererIndex=2}
  ]
 }
}
summary {
 systemName=NS_Pickaxe_SoupGoal_Trail
 emitters [
  {emitterName=FN_AnimTrail, rendererClasses [&NiagaraRibbonRendererProperties]}
  {emitterName=LightningTrail, rendererClasses [&NiagaraRibbonRendererProperties]}
 ]
 userVariables [
  User.BaseTrailColorCurve {
   type=NiagaraDataInterfaceColorCurve
   description=""
   defaultValue {
    struct=NiagaraExt_VariableValue_DataInterface
    value {
     dataInterface=&self:NiagaraDataInterfaceColorCurve_1
     dataInterfaceClass=&NiagaraDataInterfaceColorCurve
    }
   }
  }
  User.bIsFrontEnd {type=NiagaraBool}
  User.LightningColorCurve {
   type=NiagaraDataInterfaceColorCurve
   description=""
   defaultValue {
    struct=NiagaraExt_VariableValue_DataInterface
    value {
     dataInterface=&self:NiagaraDataInterfaceColorCurve_0
     dataInterfaceClass=&NiagaraDataInterfaceColorCurve
    }
   }
  }
  User.ReactivityCounter {type=NiagaraInt32}
  User.ReactivityDebug {type=NiagaraBool}
  User.ReactivityDuration {type=NiagaraFloat, defaultValue=5, description=""}
  User.ReactivityTime {type=NiagaraFloat, defaultValue=0, description=""}
  User.RibbonIDOffset {type=NiagaraInt32}
 ]
}
