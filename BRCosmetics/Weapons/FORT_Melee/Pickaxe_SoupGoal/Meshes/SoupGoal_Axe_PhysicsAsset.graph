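# Physics asset for the SoupGoal pickaxe mesh, a cosmetic melee harvesting tool.
# A single body on the b_AR_Root bone uses one capsule shape rotated to lie along the weapon and offset slightly from the bone origin.
# It has no constraints, so the whole axe is approximated as one rigid collision volume.
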PhysicsAsset /BRCosmetics/Weapons/FORT_Melee/Pickaxe_SoupGoal/Meshes/SoupGoal_Axe_PhysicsAsset
bodies {b_AR_Root [{shapeType=Capsule, length=62.6025, radius=38.2193, center=<1.8064 -1.0683 -3.4202>, rotation=<rot 0.1287 -90 -0.5271>}]}
constraints []
